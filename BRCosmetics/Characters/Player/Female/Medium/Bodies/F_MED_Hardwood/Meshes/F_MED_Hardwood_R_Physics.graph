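# Physics asset for the right-leg pants dynamics on the F_MED_Hardwood female medium character body.
# A thigh_r capsule is linked to tiny front and back base capsules through constraints with free swing and twist.
# Each base connects to a box body for its pants panel, with narrow angular limits that allow most movement on swing1, slightly more at the front than the back.
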PhysicsAsset /BRCosmetics/Characters/Player/Female/Medium/Bodies/F_MED_Hardwood/Meshes/F_MED_Hardwood_R_Physics
bodies {
 dyn_pants_bk_base_r [{shapeType=Capsule, length=0.505, radius=0.505, rotation=<rot 0 -90 180>}]
 dyn_pants_bk_r [{shapeType=Box, extentX=11.5, extentY=3, extentZ=16, center=<-5.5 0 0>, rotation=<rot 0 0 180>}]
 dyn_pants_fr_base_r [{shapeType=Capsule, length=0.505, radius=0.505, rotation=<rot 0 -90 180>}]
 dyn_pants_fr_r [{shapeType=Box, extentX=11.5, extentY=6, extentZ=16, center=<-5.5 0 0>, rotation=<rot 0 0 180>}]
 thigh_r [{shapeType=Capsule, length=8.5, radius=6.25, center=<28.9543 0.6888 0.0473>, rotation=<rot 90 0 180>}]
}
constraints [
 {
  bone1Name=dyn_pants_fr_base_r
  bone2Name=thigh_r
  swing1LimitDegrees=45
  swing1Motion=Free
  swing2LimitDegrees=45
  swing2Motion=Free
  twistLimitDegrees=45
  twistMotion=Free
 }
 {bone1Name=dyn_pants_fr_r, bone2Name=dyn_pants_fr_base_r, swing1LimitDegrees=25, swing2LimitDegrees=5, twistLimitDegrees=5}
 {
  bone1Name=dyn_pants_bk_base_r
  bone2Name=thigh_r
  swing1LimitDegrees=45
  swing1Motion=Free
  swing2LimitDegrees=45
  swing2Motion=Free
  twistLimitDegrees=45
  twistMotion=Free
 }
 {bone1Name=dyn_pants_bk_r, bone2Name=dyn_pants_bk_base_r, swing1LimitDegrees=20, swing2LimitDegrees=5, twistLimitDegrees=5}
]
