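# Animation Blueprint for the female medium Hardwood cosmetic body, copying its attached parent's pose and adding corrective deformation and pants physics.
# Each update stores elbow, hip and knee yaw deltas from the reference pose.
# These angles drive range-mapped additive tweaks on volume and pants bones, alongside weighted constraints on volume base bones.
# Left and right RigidBody nodes then simulate in thigh space; two left-pants AnimDynamics nodes sit disconnected from the output.
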AnimBlueprint /BRCosmetics/Characters/Player/Female/Medium/Bodies/F_MED_Hardwood/Meshes/F_MED_Hardwood_AnimBP
target_skeleton=&/F_MED_Hardwood_Skeleton
graphs [
 AnimGraph {nodes [
   n1 = Anim Dynamics - Bone: dyn_pants_fr_l(Alpha=0.5, ExternalForce="0,0,0") {
    angularDampingOverride=1
    angularSpringConstant=1500
    bAngularSpring=true
    bGravityOverrideInSimSpace=true
    bOverrideAngularDamping=true
    gravityScale=0.35
    simulationSpace=RootRelative
    alphaBoolBlend {blendOption=Linear}
    boundBone {boneName=dyn_pants_fr_l}
    physicsBodyDefinitions [{
      sphereCollisionRadius=0
      boundBone {boneName=dyn_pants_fr_l}
      boxExtents=<10 20 20>
      constraintSetup {angularLimitsMax=<0 0 20>, angularLimitsMin=<0 0 -5>, angularTarget=<0 0 0>}
      localJointOffset=<10 0 0>
     }]
   }
   Anim Dynamics - Bone: dyn_pants_bk_l(ComponentPose=@n1, Alpha=1, ExternalForce="0,0,0") {
    angularDampingOverride=1
    angularSpringConstant=1500
    bAngularSpring=true
    bGravityOverrideInSimSpace=true
    bOverrideAngularDamping=true
    gravityScale=0.35
    simulationSpace=RootRelative
    alphaBoolBlend {blendOption=Linear}
    boundBone {boneName=dyn_pants_bk_l}
    physicsBodyDefinitions [{
      sphereCollisionRadius=0
      boundBone {boneName=dyn_pants_bk_l}
      boxExtents=<10 20 20>
      constraintSetup {angularLimitsMax=<0 0 10>, angularLimitsMin=<0 0 -20>, angularTarget=<0 0 0>}
      localJointOffset=<10 0 0>
     }]
   }
   n2 = "Constraint - dyn_volume_base_lowerarm_l (PP)"!(
    ComponentPose=Local To Component(LocalPose=Copy Pose From Mesh() {bUseAttachedParent=true})
    Alpha=1
    ConstraintWeights_0=0.65
    ConstraintWeights_1:float=0.35
   ) -> (Pose) {
    alphaBoolBlend {blendOption=Linear}
    boneToModify {boneName=dyn_volume_base_lowerarm_l}
    constraintSetup [{targetBone {boneName=upperarm_l}}, {targetBone {boneName=lowerarm_l}}]
    constraintWeights [1,1]
   }
   n3 = "Constraint - dyn_volume_base_lowerarm_r (PP)"!(ComponentPose=@n2, Alpha=1, ConstraintWeights_0=0.65, ConstraintWeights_1:float=0.35) -> (Pose) {
    alphaBoolBlend {blendOption=Linear}
    boneToModify {boneName=dyn_volume_base_lowerarm_r}
    constraintSetup [{targetBone {boneName=upperarm_r}}, {targetBone {boneName=lowerarm_r}}]
    constraintWeights [1,1]
   }
   n4 = "Transform (Modify) Bone - Bone: dyn_volume_lowerarm_l"(ComponentPose=@n3, Alpha=Get elbow_rot_l() -> (elbow_rot_l:double), Translation="0,-2,0", Rotation="0,0,0", Scale="0.3,1,1") {
    rotationMode=BMM_Additive
    rotationSpace=BCS_BoneSpace
    scaleMode=BMM_Additive
    scaleSpace=BCS_BoneSpace
    translationMode=BMM_Additive
    translationSpace=BCS_BoneSpace
    alphaBoolBlend {blendOption=Linear}
    alphaScaleBiasClamp {bMapRange=true, inRange {max=80}}
    boneToModify {boneName=dyn_volume_lowerarm_l}
    scale {x=0.3}
    translation {y=-2}
   }
   n5 = "Transform (Modify) Bone - Bone: dyn_volume_lowerarm_r"(ComponentPose=@n4, Alpha=Get elbow_rot_r() -> (elbow_rot_r:double), Translation="0,2,0", Rotation="0,0,0", Scale="0.3,1,1") {
    rotationMode=BMM_Additive
    rotationSpace=BCS_BoneSpace
    scaleMode=BMM_Additive
    scaleSpace=BCS_BoneSpace
    translationMode=BMM_Additive
    translationSpace=BCS_BoneSpace
    alphaBoolBlend {blendOption=Linear}
    alphaScaleBiasClamp {bMapRange=true, inRange {max=80}}
    boneToModify {boneName=dyn_volume_lowerarm_r}
    scale {x=0.3}
    translation {y=2}
   }
   n6 = "Constraint - dyn_volume_base_thigh_l (PP)"!(ComponentPose=@n5, Alpha=1, ConstraintWeights_0=0.5, ConstraintWeights_1:float=0.5) -> (Pose) {
    alphaBoolBlend {blendOption=Linear}
    boneToModify {boneName=dyn_volume_base_thigh_l}
    constraintSetup [{targetBone {boneName=pelvis}}, {targetBone {boneName=thigh_l}}]
    constraintWeights [1,1]
   }
   n7 = "Constraint - dyn_volume_base_thigh_r (PP)"!(ComponentPose=@n6, Alpha=1, ConstraintWeights_0=0.5, ConstraintWeights_1:float=0.5) -> (Pose) {
    alphaBoolBlend {blendOption=Linear}
    boneToModify {boneName=dyn_volume_base_thigh_r}
    constraintSetup [{targetBone {boneName=pelvis}}, {targetBone {boneName=thigh_r}}]
    constraintWeights [1,1]
   }
   n8 = Get thigh_rot_l()
   n9 = "Transform (Modify) Bone - Bone: dyn_volume_thigh_l"(ComponentPose=@n7, Alpha=@n8, Translation="-8,-3,0", Rotation="0,0,0", Scale="1,1.2,1.1") {
    rotationMode=BMM_Additive
    rotationSpace=BCS_BoneSpace
    scaleMode=BMM_Additive
    scaleSpace=BCS_BoneSpace
    translationMode=BMM_Additive
    translationSpace=BCS_BoneSpace
    alphaBoolBlend {blendOption=Linear}
    alphaScaleBiasClamp {bMapRange=true, inRange {max=0, min=-100}, outRange {max=0, min=1}}
    boneToModify {boneName=dyn_volume_thigh_l}
    scale {y=1.2, z=1.1}
    translation=<-8 -3>
   }
   n10 = "Transform (Modify) Bone - Bone: dyn_volume_thigh_l"(ComponentPose=@n9, Alpha=@n8, Translation="0,-4,0", Rotation="0,0,0", Scale="1,1,1") {
    rotationMode=BMM_Additive
    rotationSpace=BCS_BoneSpace
    scaleMode=BMM_Additive
    scaleSpace=BCS_BoneSpace
    translationMode=BMM_Additive
    translationSpace=BCS_BoneSpace
    alphaBoolBlend {blendOption=Linear}
    alphaScaleBiasClamp {bMapRange=true, inRange {max=50}}
    boneToModify {boneName=dyn_volume_thigh_l}
    translation {y=-4}
   }
   n11 = Get thigh_rot_r()
   n12 = "Transform (Modify) Bone - Bone: dyn_volume_thigh_r"(ComponentPose=@n10, Alpha=@n11, Translation="8,3,0", Rotation="0,0,0", Scale="1,1.2,1") {
    rotationMode=BMM_Additive
    rotationSpace=BCS_BoneSpace
    scaleMode=BMM_Additive
    scaleSpace=BCS_BoneSpace
    translationMode=BMM_Additive
    translationSpace=BCS_BoneSpace
    alphaBoolBlend {blendOption=Linear}
    alphaScaleBiasClamp {bMapRange=true, inRange {max=0, min=-100}, outRange {max=0, min=1}}
    boneToModify {boneName=dyn_volume_thigh_r}
    scale {y=1.2}
    translation=<8 3>
   }
   n13 = "Transform (Modify) Bone - Bone: dyn_volume_thigh_r"(ComponentPose=@n12, Alpha=@n11, Translation="0,4,0", Rotation="0,0,0", Scale="1,1,1") {
    rotationMode=BMM_Additive
    rotationSpace=BCS_BoneSpace
    scaleMode=BMM_Additive
    scaleSpace=BCS_BoneSpace
    translationMode=BMM_Additive
    translationSpace=BCS_BoneSpace
    alphaBoolBlend {blendOption=Linear}
    alphaScaleBiasClamp {bMapRange=true, inRange {max=50}}
    boneToModify {boneName=dyn_volume_thigh_r}
    translation {y=4}
   }
   n14 = "Constraint - dyn_volume_base_thigh_fr_l (PP)"!(ComponentPose=@n13, Alpha=1, ConstraintWeights_0=0.3, ConstraintWeights_1:float=0.7) -> (Pose) {
    alphaBoolBlend {blendOption=Linear}
    boneToModify {boneName=dyn_volume_base_thigh_fr_l}
    constraintSetup [{targetBone {boneName=pelvis}}, {targetBone {boneName=thigh_l}}]
    constraintWeights [1,1]
   }
   n15 = "Constraint - dyn_volume_base_thigh_fr_r (PP)"!(ComponentPose=@n14, Alpha=1, ConstraintWeights_0=0.3, ConstraintWeights_1:float=0.7) -> (Pose) {
    alphaBoolBlend {blendOption=Linear}
    boneToModify {boneName=dyn_volume_base_thigh_fr_r}
    constraintSetup [{targetBone {boneName=pelvis}}, {targetBone {boneName=thigh_r}}]
    constraintWeights [1,1]
   }
   n16 = "Transform (Modify) Bone - Bone: dyn_volume_thigh_fr_l"(ComponentPose=@n15, Alpha=Get thigh_rot_l(), Translation="12,0,0", Rotation="0,0,0", Scale="1,0.5,1") {
    rotationMode=BMM_Additive
    rotationSpace=BCS_BoneSpace
    scaleMode=BMM_Additive
    scaleSpace=BCS_BoneSpace
    translationMode=BMM_Additive
    translationSpace=BCS_BoneSpace
    alphaBoolBlend {blendOption=Linear}
    alphaScaleBiasClamp {bMapRange=true, inRange {max=0, min=-100}, outRange {max=0, min=1}}
    boneToModify {boneName=dyn_volume_thigh_fr_l}
    scale {y=0.5}
    translation {x=12}
   }
   n17 = "Transform (Modify) Bone - Bone: dyn_volume_thigh_fr_r"(ComponentPose=@n16, Alpha=Get thigh_rot_r(), Translation="-12,0,0", Rotation="0,0,0", Scale="1,0.5,1") {
    rotationMode=BMM_Additive
    rotationSpace=BCS_BoneSpace
    scaleMode=BMM_Additive
    scaleSpace=BCS_BoneSpace
    translationMode=BMM_Additive
    translationSpace=BCS_BoneSpace
    alphaBoolBlend {blendOption=Linear}
    alphaScaleBiasClamp {bMapRange=true, inRange {max=0, min=-100}, outRange {max=0, min=1}}
    boneToModify {boneName=dyn_volume_thigh_fr_r}
    scale {y=0.5}
    translation {x=-12}
   }
   n18 = "Constraint - dyn_volume_base_calf_l (PP)"!(ComponentPose=@n17, Alpha=1, ConstraintWeights_0=0.5, ConstraintWeights_1:float=0.5) -> (Pose) {
    alphaBoolBlend {blendOption=Linear}
    boneToModify {boneName=dyn_volume_base_calf_l}
    constraintSetup [{targetBone {boneName=thigh_l}}, {targetBone {boneName=calf_l}}]
    constraintWeights [1,1]
   }
   n19 = "Constraint - dyn_volume_base_calf_r (PP)"!(ComponentPose=@n18, Alpha=1, ConstraintWeights_0=0.5, ConstraintWeights_1:float=0.5) -> (Pose) {
    alphaBoolBlend {blendOption=Linear}
    boneToModify {boneName=dyn_volume_base_calf_r}
    constraintSetup [{targetBone {boneName=thigh_r}}, {targetBone {boneName=calf_r}}]
    constraintWeights [1,1]
   }
   n20 = "Transform (Modify) Bone - Bone: dyn_volume_calf_l"(ComponentPose=@n19, Alpha=Get knee_rot_l() -> (knee_rot_l:double), Translation="1,1,0", Rotation="0,0,0", Scale="1.2,1,1.2") {
    rotationMode=BMM_Additive
    rotationSpace=BCS_BoneSpace
    scaleMode=BMM_Additive
    scaleSpace=BCS_BoneSpace
    translationMode=BMM_Additive
    translationSpace=BCS_BoneSpace
    alphaBoolBlend {blendOption=Linear}
    alphaScaleBiasClamp {bMapRange=true, inRange {max=80}}
    boneToModify {boneName=dyn_volume_calf_l}
    scale {x=1.2, z=1.2}
    translation=<1 1>
   }
   n21 = "Transform (Modify) Bone - Bone: dyn_volume_calf_r"(ComponentPose=@n20, Alpha=Get knee_rot_r() -> (knee_rot_r:double), Translation="-1,-1,0", Rotation="0,0,0", Scale="1.2,1,1.2") {
    rotationMode=BMM_Additive
    rotationSpace=BCS_BoneSpace
    scaleMode=BMM_Additive
    scaleSpace=BCS_BoneSpace
    translationMode=BMM_Additive
    translationSpace=BCS_BoneSpace
    alphaBoolBlend {blendOption=Linear}
    alphaScaleBiasClamp {bMapRange=true, inRange {max=80}}
    boneToModify {boneName=dyn_volume_calf_r}
    scale {x=1.2, z=1.2}
    translation=<-1 -1>
   }
   n22 = Get knee_rot_l()
   n23 = "Transform (Modify) Bone - Bone: dyn_pants_fr_base_l"(ComponentPose=@n21, Alpha=@n22, Translation="-1,0,0", Rotation="0,-10,0", Scale="1,1,1") {
    rotationMode=BMM_Additive
    rotationSpace=BCS_BoneSpace
    scaleMode=BMM_Additive
    scaleSpace=BCS_BoneSpace
    translationMode=BMM_Additive
    translationSpace=BCS_BoneSpace
    alphaBoolBlend {blendOption=Linear}
    alphaScaleBiasClamp {bMapRange=true, inRange {max=80}}
    boneToModify {boneName=dyn_pants_fr_base_l}
    rotation {yaw=-10}
    translation {x=-1}
   }
   n24 = "Transform (Modify) Bone - Bone: dyn_pants_bk_base_l"(ComponentPose=@n23, Alpha=@n22, Translation="-4,0,0", Rotation="0,5,0", Scale="1,1,1") {
    rotationMode=BMM_Additive
    rotationSpace=BCS_BoneSpace
    scaleMode=BMM_Additive
    scaleSpace=BCS_BoneSpace
    translationMode=BMM_Additive
    translationSpace=BCS_BoneSpace
    alphaBoolBlend {blendOption=Linear}
    alphaScaleBiasClamp {bMapRange=true, inRange {max=80}}
    boneToModify {boneName=dyn_pants_bk_base_l}
    rotation {yaw=5}
    translation {x=-4}
   }
   n25 = Get knee_rot_r()
   n26 = "Transform (Modify) Bone - Bone: dyn_pants_fr_base_r"(ComponentPose=@n24, Alpha=@n25, Translation="1,0,0", Rotation="0,-10,0", Scale="1,1,1") {
    rotationMode=BMM_Additive
    rotationSpace=BCS_BoneSpace
    scaleMode=BMM_Additive
    scaleSpace=BCS_BoneSpace
    translationMode=BMM_Additive
    translationSpace=BCS_BoneSpace
    alphaBoolBlend {blendOption=Linear}
    alphaScaleBiasClamp {bMapRange=true, inRange {max=80}}
    boneToModify {boneName=dyn_pants_fr_base_r}
    rotation {yaw=-10}
    translation {x=1}
   }
   n27 = "Transform (Modify) Bone - Bone: dyn_pants_bk_base_r"(ComponentPose=@n26, Alpha=@n25, Translation="6,0,0", Rotation="0,5,0", Scale="1,1,1") {
    rotationMode=BMM_Additive
    rotationSpace=BCS_BoneSpace
    scaleMode=BMM_Additive
    scaleSpace=BCS_BoneSpace
    translationMode=BMM_Additive
    translationSpace=BCS_BoneSpace
    alphaBoolBlend {blendOption=Linear}
    alphaScaleBiasClamp {bMapRange=true, inRange {max=80}}
    boneToModify {boneName=dyn_pants_bk_base_r}
    rotation {yaw=5}
    translation {x=6}
   }
   n28 = RigidBody(ComponentPose=@n27) {
    bClampLinearTranslationLimitToRefPose=true
    bOverrideWorldGravity=true
    overridePhysicsAsset=&/F_MED_Hardwood_L_Physics
    simulationSpace=BaseBoneSpace
    alphaBoolBlend {blendOption=Linear}
    baseBoneRef {boneName=thigh_l}
    componentLinearAccScale=<0.3 0.3 0.3>
    componentLinearVelScale=<0.3 0.3 0.3>
    overrideWorldGravity {z=-980}
   }
   n29 = RigidBody(ComponentPose=@n28) {
    bClampLinearTranslationLimitToRefPose=true
    bOverrideWorldGravity=true
    overridePhysicsAsset=&/F_MED_Hardwood_R_Physics
    simulationSpace=BaseBoneSpace
    alphaBoolBlend {blendOption=Linear}
    baseBoneRef {boneName=thigh_r}
    componentLinearAccScale=<0.3 0.3 0.3>
    componentLinearVelScale=<0.3 0.3 0.3>
    overrideWorldGravity {z=-980}
   }
   Output Pose(Result=Component To Local(ComponentPose=@n29)) {layerGroup=None}
  ]}
 EventGraph {nodes [
   Event BlueprintUpdateAnimation()
   > n1 = Sequence!(execute) -> (then_0, then_1, then_2:Exec, then_3:Exec, then_4:Exec, then_5:Exec, then_6:Exec)
   n2 = GetDeltaTransformFromRefPose(self=GetOwnerSkeletalMeshComponent(), BoneName=lowerarm_l, BaseName=upperarm_l)
   > Set elbow_rot_l(elbow_rot_l:double=BreakRotator(InRot=BreakTransform(InTransform=@n2).Rotation).Yaw) -> (Output_Get:double)
   n3 = GetDeltaTransformFromRefPose(self=GetOwnerSkeletalMeshComponent(), BoneName=thigh_r, BaseName=pelvis)
   Set thigh_rot_r(execute=@n1.then_3, thigh_rot_r=BreakRotator(InRot=BreakTransform(InTransform=@n3).Rotation).Yaw)
   n4 = GetDeltaTransformFromRefPose(self=GetOwnerSkeletalMeshComponent(), BoneName=calf_r, BaseName=thigh_r)
   Set knee_rot_r(execute=@n1.then_5, knee_rot_r:double=BreakRotator(InRot=BreakTransform(InTransform=@n4).Rotation).Yaw) -> (Output_Get:double)
   n5 = GetDeltaTransformFromRefPose(self=GetOwnerSkeletalMeshComponent(), BoneName=calf_l, BaseName=thigh_l)
   Set knee_rot_l(execute=@n1.then_4, knee_rot_l:double=BreakRotator(InRot=BreakTransform(InTransform=@n5).Rotation).Yaw) -> (Output_Get:double)
   n6 = GetDeltaTransformFromRefPose(self=GetOwnerSkeletalMeshComponent(), BoneName=thigh_l, BaseName=pelvis)
   Set thigh_rot_l(execute=@n1.then_2, thigh_rot_l=BreakRotator(InRot=BreakTransform(InTransform=@n6).Rotation).Yaw)
   n7 = GetDeltaTransformFromRefPose(self=GetOwnerSkeletalMeshComponent(), BoneName=lowerarm_r, BaseName=upperarm_r)
   Set elbow_rot_r(execute=@n1.then_1, elbow_rot_r:double=BreakRotator(InRot=BreakTransform(InTransform=@n7).Rotation).Yaw) -> (Output_Get:double)
  ]}
]
variables [
 elbow_rot_l {type=number}
 elbow_rot_r {type=number}
 knee_rot_l {type=number}
 knee_rot_r {type=number}
 thigh_rot_l {type=number}
 thigh_rot_r {type=number}
]
